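# Masked two-sided material for the SwoopClasp gadget animation prop that adds scrolling static lines to its emissive, built between the shared 2023 character PreFX and PostFX functions.
# Three upward or downward panning samples of swoop and static textures combine into lines that boost the optionally desaturated emissive between low and high intensities.
# A mask texture blends horizontal and vertical planar UVs from pre-skinned position.
# The opacity mask keeps only back faces, while shadows stay solid.
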Material /BRCosmetics/Animation/Props/Gadgets_SwoopClasp/Materials/M_SwoopClasp
graph {
 properties {
  bUseMaterialAttributes=true
  bUsedWithMeshDeformer=true
  bUsedWithMorphTargets=true
  bUsedWithNiagaraMeshParticles=true
  bUsedWithSkeletalMesh=true
  blendMode=BLEND_Masked
  pixelDepthOffsetMode=PDOM_Legacy
  shadingModel=MSM_FromMaterialExpression
  twoSided=true
  usedShadingModels=Default Lit | Subsurface | Clear Coat | Subsurface Profile | Cloth
 }
 nodes [
  n1 = MaterialFunctionCall() {materialFunction=&MF_CharacterParent_2023_PreFX}
  n2 = GetMaterialAttributes(None=@n1.Result) {attributeGetTypes [#1]}
  n3 = Desaturation(
   None=NamedRerouteDeclaration(None=@n2.Emissive Color) {name=EmissiveInput, nodeColor=<rgba 1 0 0.3281 1>}
   Fraction=ScalarParameter() {parameterName=ColorDesaturation}
  )
  n4 = ScalarParameter() {defaultValue=1, parameterName=LineIntensityLOW}
  n5 = ScalarParameter() {defaultValue=10, parameterName=LineIntensityHigh}
  n6 = NamedRerouteUsage()
  n7 = Add(A=Constant() {r=1}, B=Multiply(A=NamedRerouteUsage(), B=LinearInterpolate(A=@n4, B=@n5, Alpha=@n6)))
  n8 = TwoSidedSign()
  n9 = OneMinus(None=Saturate(None=Sign(None=@n8)))
  n10 = Constant() {r=1}
  MaterialFunctionCall(
   Material=SetMaterialAttributes!(MaterialAttributes=@n2.MaterialAttributes, Emissive Color=Multiply(A=@n3, B=@n7), Opacity Mask=ShadowReplace(Default=@n9, Shadow=@n10)) {attributeSetTypes [#1, #2]}
   M=@n1.M
   Pre Skinned Local Position=@n1.M
   Distance Blend=@n1.M
   Fresnel=@n1.M
  ) {materialFunction=&MF_CharacterParent_2023_PostFX}
  n11 = Multiply(A=NamedRerouteUsage(), B=ScalarParameter() {defaultValue=1, parameterName=LineTiling}) {constB=2}
  n12 = TextureSample(UVs=Panner(Coordinate=@n11) {speedY=0.1}) {samplerSource=SSM_Wrap_WorldGroupSettings, texture=/BRCosmetics/Animation/Props/Gadgets_SwoopClasp/Textures/T_EmoteSwoop}
  n13 = TextureSample(UVs=Panner(Coordinate=@n11) {speedY=0.05}) {
   samplerSource=SSM_Wrap_WorldGroupSettings
   samplerType=SAMPLERTYPE_Masks
   texture=/Game/Missions/Primary/RepairTheShelter/Textures/Cinematic_Screen_Static
  }
  n14 = TextureSample(UVs=Panner(Coordinate=@n11) {speedY=0.15}) {samplerSource=SSM_Wrap_WorldGroupSettings, texture=&T_EmoteSwoop}
  n15 = Multiply(
   A=Add(A=Multiply(A=@n12.B) {constB=0.2}, B=Multiply(A=@n13.G, B=@n14.R))
   B=ScalarParameter() {defaultValue=3, group=TV Glitches, parameterName=Static Scale}
  )
  NamedRerouteDeclaration(None=Saturate(None=NamedRerouteDeclaration(None=@n15) {name=Static, nodeColor=<rgba 1 0.6562 0 1>})) {name=PanningLines, nodeColor=<rgba 0 0.6484 1 1>}
  NamedRerouteDeclaration(None=@n1.Pre skinned Local Position) {name=PreskinnedPos, nodeColor=<rgba 0 1 0.2266 1>}
  n16 = NamedRerouteUsage()
  n17 = Divide(A=@n16, B=ScalarParameter() {defaultValue=100, parameterName=PreskinnedUnitHoriztonal})
  n18 = Divide(A=@n16, B=ScalarParameter() {defaultValue=30, parameterName=PreskinnedUnitVertical})
  n19 = TextureSampleParameter2D() {
   group=ScreenEffect
   parameterName=TechArtMaskRE
   samplerType=SAMPLERTYPE_Masks
   texture=/BRCosmetics/Animation/Props/Gadgets_SwoopClasp/Textures/T_emoteswoop_Emote_Prop_FX
  }
  NamedRerouteDeclaration(None=LinearInterpolate(A=ComponentMask(None=@n17) {g=true, r=true}, B=ComponentMask(None=@n18) {b=true, r=true}, Alpha=@n19.R)) {name=PreskinnedUV, nodeColor=<rgba 1 0 0.0937 1>}
  NamedRerouteDeclaration(None=@n19.R) {name=VerticalElement, nodeColor=<rgba 0 0.2734 1 1>}
  TextureCoordinate()
 ]
}
