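# Physics asset for the Swoop Clasp gadget prop mesh in BRCosmetics, giving its skeleton simple collision bodies.
# The main and root bones each carry one large capsule whose radius far exceeds its length, so both are close to spherical.
# A single constraint links main to root with matching angular limits on both swings and twist.
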PhysicsAsset /BRCosmetics/Animation/Props/Gadgets_SwoopClasp/Meshes/PA_SwoopClasp_Gadget
bodies {
 main [{shapeType=Capsule, length=56.2946, radius=217.5417, center=<-0.0001 -22.1688 5.924>, rotation=<rot 0 89.9998 -90>}]
 root [{shapeType=Capsule, length=0.505, radius=247.0958, rotation=<rot 0 -90 0>}]
}
constraints [{bone1Name=main, bone2Name=root, swing1LimitDegrees=45, swing2LimitDegrees=45, twistLimitDegrees=45}]
